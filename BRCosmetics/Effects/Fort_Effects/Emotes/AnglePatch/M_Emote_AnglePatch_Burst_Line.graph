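# Translucent, unlit, two-sided Niagara particle material for the AnglePatch emote burst line, tinted by color parameter, particle color and vertex red.
# Opacity scrolls a line texture through radial coordinates via the Sweep dynamic parameter.
# Parameters blend scanline, flickering mosaic and second-UV sweep variants; the default selects the last, despite Use_Scanlines being 1.
# A sphere mask trims opacity, depth fade softens higher-quality non-mobile paths, and SplineThicken offsets are blendable.
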Material /BRCosmetics/Effects/Fort_Effects/Emotes/AnglePatch/M_Emote_AnglePatch_Burst_Line
graph {
 properties {
  bAutomaticallySetUsageInEditor=false
  bUsedWithNiagaraMeshParticles=true
  bUsedWithNiagaraSprites=true
  blendMode=BLEND_Translucent
  shadingModel=MSM_Unlit
  twoSided=true
  usedShadingModels=Unlit
 }
 nodes [
  LinearInterpolate(
   A=MaterialFunctionCall(
    WidthBase=ScalarParameter() {defaultValue=0.4, parameterName=SplineBase}
    WidthTip=ScalarParameter() {defaultValue=0.4, parameterName=SplineTip}
    UVs for Thickness=TextureCoordinate()
    Expand U or V UV Channel=StaticBool()
   ) {materialFunction=&SplineThicken}.WorldPositionOffset
   B=Constant()
   Alpha=ScalarParameter() {parameterName=Use_SplineThicken}
  )
  n1 = VectorParameter() {parameterName=Color, defaultValue=<rgba 0.2344 2 0.0706 1>}
  n2 = ParticleColor()
  MaterialFunctionCall(Emissive=Multiply(A=Multiply(A=@n1.RGB, B=@n2.RGB), B=VertexColor().R)) {materialFunction=&MF_FX_EmissiveAdjustments}
  n3 = ComponentMask(None=TextureCoordinate()) {g=true}
  n4 = TextureCoordinate()
  n5 = Power(Base=OneMinus(None=ComponentMask(None=@n4) {g=true}), Exp=ScalarParameter() {defaultValue=1, parameterName=Mask_Edge_power}) {constExponent=1}
  n6 = ScalarParameter() {defaultValue=3, parameterName=Mask Edge intensity}
  Multiply(A=Saturate(None=Multiply(A=@n3) {constB=5}), B=Saturate(None=Multiply(A=@n5, B=@n6) {constB=3}))
  n7 = TextureCoordinate()
  n8 = Add(A=@n7, B=Constant2Vector() {g=-0.5, r=-0.5})
  n9 = AppendVector(
   A=Constant()
   B=Multiply(A=DynamicParameter() {defaultValue=<rgba -1 1 1 1>, paramNames [Sweep,Param2,Param3,Param4]}.Sweep, B=VertexColor().R)
  )
  n10 = TextureSampleParameter2D!(
   UVs=Add(
    A=MaterialFunctionCall(Vector or UVs=@n8) {materialFunction=&VectorToRadialValue}.Radial Coordinates
    B=@n9
   )
   "DDX(UVs)"=DDX(Value=@n7)
   "DDY(UVs)"=DDY(Value=@n7)
   Apply View MipBias
  ) {
   mipValueMode=TMVM_Derivative
   parameterName=SweepTexture
   samplerSource=SSM_Clamp_WorldGroupSettings
   texture=&/T_Emote_AnglePatch_Line
  }
  n11 = LinearInterpolate(Alpha=@n10.R)
  n12 = Multiply(
   A=@n11
   B=MaterialFunctionCall!(
    Number of Scanlines=ScalarParameter() {defaultValue=150, parameterName=Scanline Amount}
    Scanline Speed
    "Scanlines Thickness (Bias)"=ScalarParameter() {defaultValue=-0.15, parameterName=ScanlineSize}
    Scanlines Dark Value=ScalarParameter() {defaultValue=0.5, parameterName=Scanlines Dark}
    Scanlines Light Value=ScalarParameter() {defaultValue=1, parameterName=Scanlines Light}
    Block Opacity
    UVs
    Time
    Block Repeat
   ) {materialFunction=/Game/Effects/Fort_Effects/Effects/Atlas/Materials/MF_Atlas_Scanlines_01}.Result
  )
  n13 = ScalarParameter() {defaultValue=2, parameterName=Mosaic_Tiling}
  n14 = Multiply(A=TextureCoordinate(), B=@n13) {constB=2}
  n15 = Time()
  n16 = Sine(None=Add(
    A=TextureSample(UVs=@n14) {samplerType=SAMPLERTYPE_LinearColor, texture=/Game/TheOutpost/Textures/AnimatedOutpost_Base046_rgb_XVector_a_NormalizedHierPos_UV_2}.R
    B=Multiply(A=@n15) {constB=2}
   )) {period=1.5}
  n17 = TextureSample(UVs=Multiply(A=TextureCoordinate() {uTiling=0.5}, B=@n13) {constB=2}) {samplerType=SAMPLERTYPE_LinearColor, texture=&AnimatedOutpost_Base046_rgb_XVector_a_NormalizedHierPos_UV_2}
  n18 = Time()
  n19 = Sine(None=Add(A=Power(Base=@n17.B), B=Multiply(A=@n18) {constB=3}))
  n20 = Multiply(A=@n2.A, B=Multiply(A=@n11, B=Multiply(A=@n16, B=@n19)))
  n21 = TextureCoordinate() {coordinateIndex=1}
  n22 = Add(A=@n21, B=Constant2Vector() {g=-0.5, r=-0.5})
  n23 = TextureSample!(
   UVs=Add(
    A=MaterialFunctionCall(Vector or UVs=@n22) {materialFunction=&VectorToRadialValue}.Radial Coordinates
    B=@n9
   )
   Tex
   "DDX(UVs)"=DDX(Value=@n21)
   "DDY(UVs)"=DDY(Value=@n21)
   Apply View MipBias
  ) {
   mipValueMode=TMVM_Derivative
   samplerSource=SSM_Clamp_WorldGroupSettings
   texture=&T_Emote_AnglePatch_Line
  }
  n24 = Multiply(A=@n2.A, B=LinearInterpolate(Alpha=@n23.R))
  n25 = ScalarParameter() {defaultValue=1, parameterName=Use_BoxSweep}
  n26 = Multiply(
   A=LinearInterpolate(
    A=Multiply(A=@n2.A, B=@n12)
    B=LinearInterpolate(A=@n20, B=@n24, Alpha=@n25)
    Alpha=ScalarParameter() {defaultValue=1, parameterName=Use_Scanlines}
   )
   B=SphereMask(A=TextureCoordinate() {coordinateIndex=1}, B=Constant2Vector() {g=0.5, r=0.5}, Radius=Constant() {r=0.5}, Hardness=Constant() {r=0.1})
  )
  n27 = DepthFade(Opacity=@n26, FadeDistance=ScalarParameter() {parameterName=DepthFade})
  n28 = ShadingPathSwitch(Default=@n27, Deferred=@n27, Forward=@n26, Mobile=@n26)
  QualitySwitch(Default=@n28, Low=@n26, High=@n28, Medium=@n26, Epic=@n28)
 ]
}
