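# Unlit, two-sided translucent material for Niagara sprite and mesh particles that draws animated, broken line patterns for the AnglePatch emote.
# One line variant floors sine waves into segments or noisy dots, masked around particle world position.
# The Line02 variant combines a ramp, panned by dynamic parameter, with noise-distorted sine waves.
# Scalar lerps select variants and optional SplineThicken vertex offset; particle color drives tint and alpha.
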Material /BRCosmetics/Effects/Fort_Effects/Emotes/AnglePatch/M_Emote_AnglePatch_Line
graph {
 properties {
  bAutomaticallySetUsageInEditor=false
  bUsedWithNiagaraMeshParticles=true
  bUsedWithNiagaraSprites=true
  blendMode=BLEND_Translucent
  shadingModel=MSM_Unlit
  twoSided=true
  usedShadingModels=Unlit
 }
 nodes [
  LinearInterpolate(
   A=MaterialFunctionCall(
    WidthBase=ScalarParameter() {defaultValue=0.4, parameterName=SplineBase}
    WidthTip=ScalarParameter() {defaultValue=0.4, parameterName=SplineTip}
    Expand U or V UV Channel=StaticBool()
   ) {materialFunction=&SplineThicken}.WorldPositionOffset
   B=Constant()
   Alpha=ScalarParameter() {parameterName=Use_SplineThicken}
  )
  n1 = ParticleColor()
  n2 = Constant() {r=2}
  n3 = VertexColor()
  n4 = Multiply(A=TextureCoordinate(), B=AppendVector(A=Multiply(A=@n2, B=@n3.R), B=Constant() {r=1}))
  n5 = Sine(None=Add(A=ComponentMask(None=@n4) {r=true}, B=@n3.R))
  n6 = ScalarParameter() {defaultValue=10, parameterName=SpherePoint_U}
  n7 = Multiply(A=TextureCoordinate() {uTiling=3}, B=AppendVector(A=Multiply(A=@n6, B=@n3.R), B=Constant() {r=1}))
  n8 = Constant() {r=2}
  n9 = Multiply(
   A=Floor(None=Add(A=@n5) {constB=0.1})
   B=SphereMask(A=Fmod(A=@n7, B=@n8), B=Constant2Vector() {g=0.5, r=0.5}, Radius=Constant() {r=0.5}, Hardness=Constant() {r=1})
  )
  n10 = TextureCoordinate() {uTiling=10, vTiling=0.01}
  n11 = AppendVector(A=Multiply(A=Constant() {r=0.5}, B=@n3.R), B=Constant())
  n12 = LinearInterpolate(Alpha=TextureSample(UVs=Panner(Coordinate=@n10, Speed=@n11) {speedX=0.5}) {samplerType=SAMPLERTYPE_Grayscale, texture=&LowResBlurredNoise}.R) {constA=-0.1}
  n13 = TextureCoordinate() {uTiling=3}
  n14 = Multiply(A=Time())
  n15 = VertexColor()
  n16 = Sine(None=Add(A=ComponentMask(None=@n13) {r=true}, B=Multiply(A=@n14, B=@n15.R)))
  n17 = TextureCoordinate() {uTiling=2}
  n18 = Multiply(A=Time()) {constB=0.2}
  n19 = VertexColor()
  n20 = Sine(None=Add(A=ComponentMask(None=@n17) {r=true}, B=Multiply(A=@n18, B=@n19.R)))
  n21 = Add(A=Floor(None=Add(A=@n16) {constB=0.05}), B=Floor(None=Add(A=@n20) {constB=0.12}))
  n22 = TextureCoordinate() {uTiling=0.5, vTiling=0.5}
  n23 = Multiply(A=Sine(None=ComponentMask(None=@n22) {g=true}), B=Sine(None=ComponentMask(None=@n22) {r=true}))
  n24 = Saturate(None=Power(Base=Multiply(A=@n23) {constB=5}))
  n25 = Multiply(
   A=LinearInterpolate(A=Multiply(A=@n9, B=@n12), B=Multiply(A=@n21, B=@n24), Alpha=ScalarParameter() {defaultValue=1, parameterName=Use Sphere})
   B=SphereMask(A=WorldPosition().XYZ, B=ParticlePositionWS(), Radius=ScalarParameter() {defaultValue=60, parameterName=WorldPosition_SphereMaskRadius}) {attenuationRadius=65, hardnessPercent=10}
  )
  n26 = TextureCoordinate() {coordinateIndex=1}
  n27 = Constant2Vector() {g=0.5, r=0.5}
  n28 = Constant() {r=0.3}
  n29 = Constant() {r=0.8}
  n30 = Multiply(A=@n25, B=OneMinus(None=SphereMask(A=@n26, B=@n27, Radius=@n28, Hardness=@n29) {attenuationRadius=65, hardnessPercent=10}))
  n31 = ScalarParameter() {defaultValue=1, parameterName=Use_CenterMask}
  LinearInterpolate(
   A=Multiply(A=@n1.A, B=NamedRerouteUsage())
   B=Multiply(A=@n1.A, B=LinearInterpolate(A=@n30, B=@n25, Alpha=@n31))
   Alpha=ScalarParameter() {defaultValue=1, parameterName=Use_SecondLine}
  )
  MaterialFunctionCall(Emissive=Multiply(A=VectorParameter() {parameterName=Color, defaultValue=<rgba 0.0729 2 0.2014 1>}.RGB, B=@n1.RGB)) {materialFunction=&MF_FX_EmissiveAdjustments}
  n32 = TextureCoordinate()
  n33 = DynamicParameter() {defaultValue=<rgba 0 1 1 1>, paramNames [MaskPanner, MaskPanner_Offset, Param3, Param4]}
  n34 = AppendVector(A=Multiply(A=@n33.MaskPanner, B=@n33.MaskPanner), B=Constant())
  n35 = TextureSample(UVs=MaterialFunctionCall(UVs=Add(A=@n32, B=@n34), "Rotation Angle (0-1)"=Constant() {r=0.25}) {materialFunction=&CustomRotator}.Rotated Values) {samplerSource=SSM_Clamp_WorldGroupSettings, texture=/Game/Effects/Fort_Effects/Textures/Ramps/T_Power_Source_Sphere_Ramp}
  n36 = ComponentMask(None=TextureCoordinate() {uTiling=0.5}) {r=true}
  n37 = Time()
  n38 = TextureSampleParameter2D(UVs=Panner(Coordinate=TextureCoordinate() {uTiling=0.5, vTiling=0.5}) {speedX=0.2}) {parameterName=NoiseLine_Tex, samplerType=SAMPLERTYPE_LinearColor, texture=/Game/Global/Textures/Default/Noise/T_Atlas_DigitalNoise_Small}
  n39 = Sine(None=Multiply(A=Sine(None=@n36), B=Add(A=@n37, B=@n38.RGB))) {period=0.5}
  n40 = ComponentMask(None=TextureCoordinate() {uTiling=3}) {r=true}
  n41 = Multiply(A=Time())
  n42 = Add(A=Sine(None=Add(A=@n40, B=@n41))) {constB=0.9}
  NamedRerouteDeclaration(None=Multiply(A=Multiply(A=@n35.R, B=@n39), B=Power(Base=@n42) {constExponent=1.1})) {name=Line02, nodeColor=<rgba 0 0.4844 1 1>}
 ]
}
